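; BP_PuzzleDoor is a prototype networked puzzle door that slides open once when OpenPuzzleDoor is called.
; BeginPlay stores the mesh's start location and adds an opening vector to get the end location.
; On authority, after a delay and a DoOnce, a timeline lerps the mesh between them in world space.
; It also calls ClientFX, which plays the door audio only on its non-authority branch.
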(event EventBeginPlay
  ; Init
  (bind _output_get (Variables|Default|SetStartLoc (Transformation|GetWorldLocation (Variables|Default|GetDoorMesh))))
  (Variables|Default|SetEndLoc (+ _output_get (Variables|Default|GetDoorOpeningVector))))

(event Custom|OpenPuzzleDoor
  ; Open Door
  (Networking|SwitchHasAuthority
    (:Authority
      (Utilities|FlowControl|Delay (Variables|Default|GetInitialDoorOpeningDelay))
      (Utilities|FlowControl|DoOnce false
        (:Completed
          (Utilities|FlowControl|Sequence
            (:then_0
              (bind (_direction _doorslide) (Timeline|OpenDoor PlayFromStart)
                (:Update
                  (Transformation|SetWorldLocation (Variables|Default|GetDoorMesh) (Math|Vector|Lerp(Vector) (Variables|Default|GetStartLoc) (Variables|Default|GetEndLoc) _doorslide) false false))))
            (:then_1
              (CallFunction|ClientFX))))))))

(event Custom|ClientFX
  ; Client FX and sounds
  (Networking|SwitchHasAuthority
    (:Remote
      (Audio|Components|Audio|Play (Variables|Default|GetAudio) 0.0))))

(fn OnRep_EndLoc)

(fn ConstructionScript)
