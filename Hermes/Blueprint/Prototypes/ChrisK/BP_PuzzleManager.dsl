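; BP_PuzzleManager is a prototype manager for a turner combination puzzle that opens a door when solved.
; On BeginPlay it caches each turner's solve index as the target combination.
; The PuzzleTurnerUpdated event compares every turner's current puzzle data against that array by index.
; On a full match it marks each turner solved, calls OpenPuzzleDoor inside that same per-turner loop, and prints a debug message.
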(event EventBeginPlay
  (CallFunction|SetCombination))

(event Custom|PuzzleTurnerUpdated
  (bind _succeeded (CallFunction|CheckCombinations))
  (if _succeeded
    (CallFunction|PuzzleIsSolvedNow)
    (Development|PrintString "Puzzle Succeeded" true true "(R=0.000000,G=0.660000,B=1.000000,A=1.000000)" 1e+01)))

(fn CheckCombinations
  (bind (_array_element _array_index) (Utilities|Array|ForEachLoopwithBreak (Variables|Default|GetAllPuzzleTurners))
    (:LoopBody
      (LocalVariables|SetCurrentIndex _array_index)
      (if (not (== (Utilities|Array|Get(aref) (Variables|Default|GetSolvedCombinations) (LocalVariables|GetCurrentIndex)) (Utilities|Struct|BreakBPPuzzleDataStruct (Variables|/Hermes/Blueprint/Prototypes/ChrisK/BP_PuzzleTurner.BP_PuzzleTurner_C|GetCurrentPuzzleData (Utilities|Array|Get(acopy) (Variables|Default|GetAllPuzzleTurners) (LocalVariables|GetCurrentIndex))))))
        (LocalVariables|SetCombinationSucceededLocal false)))
    (:Completed
      _))
  (return (LocalVariables|GetCombinationSucceededLocal)))

(fn PuzzleIsSolvedNow
  (for _array_element (Variables|Default|GetAllPuzzleTurners)
    (Class|BPPuzzleTurner|PuzzleIsSolved _array_element)
    (Class|BPPuzzleDoor|OpenPuzzleDoor (Variables|Default|GetPuzzleDoorRef))))

(fn SetCombination
  (for _array_element (Variables|Default|GetAllPuzzleTurners)
    (Utilities|Array|Add (Variables|Default|GetSolvedCombinations) (Variables|/Hermes/Blueprint/Prototypes/ChrisK/BP_PuzzleTurner.BP_PuzzleTurner_C|GetPuzzleSolveIndex _array_element))))

(fn ConstructionScript)
