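; Prototype interactable statue for a rotation puzzle, starting at a configurable index among eight yaw positions and turned by players until solved.
; Each interaction runs a timeline that lerps the mesh 45 degrees, with turn audio on clients.
; On finish, the authority maps rotation to a position index, updates debug text and notifies the puzzle manager.
; PuzzleIsSolved sets the solved flag and spawns Niagara and stinger effects on clients.
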(event EventBeginPlay
  ; Init Starting Rot
  (CallFunction|SetInitialPuzzleStartingIndex))

(event Interaction|EventBlueprintOnInteract
  (params (InteractingPawn :type Fort%20Pawn%20Object%20Reference) (InteractionBeingAttempted :type EInteractionBeingAttempted%20Enum :default "FirstInteraction"))
  ; Rotate On Interact
  (if (not (Variables|Default|GetPuzzleSolved))
    (Utilities|FlowControl|DoOnce false
      (:Completed
        (Variables|Default|SetStatueCanTurn false)
        (Variables|Default|SetStatueIsTurning true)
        (CallFunction|SetCurrentandNextRotation)
        (CallFunction|TurningClientEffects)
        (bind (_direction _turnstatue) (Timeline|TurnStatue PlayFromStart)
          (:Update
            (Transformation|SetRelativeRotation (Variables|Default|GetStatueMesh) (Math|Rotator|Lerp(Rotator) (Variables|Default|GetCurrentRelativeRotation) (Variables|Default|GetNextRelativeRotation) _turnstatue true) false false))
          (:Finished
            (Networking|SwitchHasAuthority
              (:Authority
                (Utilities|FlowControl|Delay 0.25)
                (Variables|Default|SetStatueIsTurning false)
                (Variables|Default|SetStatueCanTurn true)
                (CallFunction|SetCurrentandNextRotation)
                (CallFunction|SetCurrentPuzzleDataIndex)
                (Class|BPPuzzleManager|PuzzleTurnerUpdated (Variables|Default|GetPuzzleManagerRef))
                (CallFunction|ResetAllowTurn)
                (Development|PrintString (Utilities|String|Append "Statue Solved State = " (Utilities|String|ToString(Boolean) (== (Variables|Default|GetPuzzleSolveIndex) (Utilities|Struct|BreakBPPuzzleDataStruct (Variables|Default|GetCurrentPuzzleData))))))))))))))

(event Custom|ResetAllowTurn
  ; Rotate On Interact
  (Utilities|FlowControl|DoOnce false
    (:Completed
      (Variables|Default|SetStatueCanTurn false)
      (Variables|Default|SetStatueIsTurning true)
      (CallFunction|SetCurrentandNextRotation)
      (CallFunction|TurningClientEffects)
      (Timeline|TurnStatue PlayFromStart))))

(event Custom|TurningClientEffects
  ; Rotate On Interact
  (Networking|SwitchHasAuthority
    (:Remote
      (Audio|Components|Audio|Play (Variables|Default|GetAudioTurnCue) 0.0))))

(event Custom|PuzzleIsSolved
  (bind _defaultsceneroot (Variables|Default|GetDefaultSceneRoot))
  (bind _returnvalue (Transformation|GetWorldLocation _defaultsceneroot))
  (bind _returnvalue_2 (+ _returnvalue "0.000000,-20.000000,140.000000"))
  ; Puzzle Solved
  (Variables|Default|SetPuzzleSolved true)
  (Networking|SwitchHasAuthority
    (:Remote
      (Niagara|SpawnSystematLocation "/Hermes/Blueprint/Prototypes/ChrisK/FX/NS_PuzzleTurner_Solved.NS_PuzzleTurner_Solved" _returnvalue_2)
      (Audio|SpawnSoundatLocation "/Hermes/Audio/Wolf_Puzzle/Interactables_S34_Wolf_Puzzle_Stinger_Cue.Interactables_S34_Wolf_Puzzle_Stinger_Cue" _returnvalue_2))))

(fn BlueprintCanInteract
  (params (InteractingPawn :type Fort%20Pawn%20Object%20Reference) (InteractionBeingAttempted :type EInteractionBeingAttempted%20Enum :default "FirstInteraction") (InteractionType :type TInteractionType%20Enum :default "IT_NoInteraction"))
  (return (!= (Variables|Default|GetPuzzleSolved) (Variables|Default|GetStatueCanTurn))))

(fn BlueprintGetInteractionString
  (params (InteractingPawn :type Fort%20Pawn%20Object%20Reference) (InteractionBeingAttempted :type EInteractionBeingAttempted%20Enum :default "FirstInteraction"))
  (return "NSLOCTEXT(\"[31696A52E8C313A6E399A325CD403B52]\", \"4C4C156A4DDCB299A1472286A51330F4\", \"Turn Statue\")"))

(fn OnRep_NumberDebug
  (Rendering|Components|TextRender|SetText (Variables|Default|GetTextRender) (Variables|Default|GetNumber%20Debug)))

(fn RandomizeTheStartingIndex
  (CallFunction|SetInitialPuzzleStartingIndex))

(fn SetCurrentAndNextRotation
  (bind _output_get (Variables|Default|SetCurrentRelativeRotation (Variables|/Script/Engine.SceneComponent|GetRelativeRotation (Variables|Default|GetStatueMesh))))
  (if (== (.yaw _output_get) 315.0)
    (Variables|Default|SetNextRelativeRotation "0, 0, 0")
    (else
      (Variables|Default|SetNextRelativeRotation (Math|Rotator|CombineRotators _output_get (Math|Rotator|MakeRotator 0.0 0.0 45.0))))))

(fn SetCurrentPuzzleDataIndex
  (if (Utilities|Operators|Equal(==) :A (Variables|Default|GetCurrentRelativeRotation) :ErrorTolerance 1.0)
    (Variables|Default|SetCurrentPuzzleData (Utilities|Struct|MakeBPPuzzleDataStruct 0))
    (Variables|Default|SetNumber%20Debug "NSLOCTEXT(\"[31696A52E8C313A6E399A325CD403B52]\", \"94C558A9413FFDBE9772C8A8DE6F2170\", \"0\")")
    (Rendering|Components|TextRender|SetText (Variables|Default|GetTextRender) (Variables|Default|GetNumber%20Debug))
    (elif (Utilities|Operators|Equal(==) (Variables|Default|GetCurrentRelativeRotation) "0.000000,45.000000,0.000000" 1.0)
      (Variables|Default|SetCurrentPuzzleData (Utilities|Struct|MakeBPPuzzleDataStruct 1 "0.000000,45.000000,0.000000"))
      (Variables|Default|SetNumber%20Debug "NSLOCTEXT(\"[31696A52E8C313A6E399A325CD403B52]\", \"3770A1F04D0AF19CA38EE7B5913F04D7\", \"1\")")
      (Rendering|Components|TextRender|SetText (Variables|Default|GetTextRender) (Variables|Default|GetNumber%20Debug))
      (elif (Utilities|Operators|Equal(==) (Variables|Default|GetCurrentRelativeRotation) "0.000000,90.000000,0.000000" 1.0)
        (Variables|Default|SetCurrentPuzzleData (Utilities|Struct|MakeBPPuzzleDataStruct 2 "0.000000,90.000000,0.000000"))
        (Variables|Default|SetNumber%20Debug "NSLOCTEXT(\"[31696A52E8C313A6E399A325CD403B52]\", \"D6BE36924FD5EF96A1E0F78068982956\", \"2\")")
        (Rendering|Components|TextRender|SetText (Variables|Default|GetTextRender) (Variables|Default|GetNumber%20Debug))
        (elif (Utilities|Operators|Equal(==) (Variables|Default|GetCurrentRelativeRotation) "0.000000,135.000000,0.000000" 1.0)
          (Variables|Default|SetCurrentPuzzleData (Utilities|Struct|MakeBPPuzzleDataStruct 3 "0.000000,135.000000,0.000000"))
          (Variables|Default|SetNumber%20Debug "NSLOCTEXT(\"[31696A52E8C313A6E399A325CD403B52]\", \"4ED98D4F4AB4A6909385F38D3F764A09\", \"3\")")
          (Rendering|Components|TextRender|SetText (Variables|Default|GetTextRender) (Variables|Default|GetNumber%20Debug))
          (elif (Utilities|Operators|Equal(==) (Variables|Default|GetCurrentRelativeRotation) "0.000000,180.000000,0.000000" 1.0)
            (Variables|Default|SetCurrentPuzzleData (Utilities|Struct|MakeBPPuzzleDataStruct 4 "0.000000,180.000000,0.000000"))
            (Variables|Default|SetNumber%20Debug "NSLOCTEXT(\"[31696A52E8C313A6E399A325CD403B52]\", \"3C35CE70427EC7812577588988B7DAD9\", \"4\")")
            (Rendering|Components|TextRender|SetText (Variables|Default|GetTextRender) (Variables|Default|GetNumber%20Debug))
            (elif (Utilities|Operators|Equal(==) (Variables|Default|GetCurrentRelativeRotation) "0.000000,225.000000,0.000000" 1.0)
              (Variables|Default|SetCurrentPuzzleData (Utilities|Struct|MakeBPPuzzleDataStruct 5 "0.000000,225.000000,0.000000"))
              (Variables|Default|SetNumber%20Debug "NSLOCTEXT(\"[31696A52E8C313A6E399A325CD403B52]\", \"AD77190546486E9FBCF97587A27D75B6\", \"5\")")
              (Rendering|Components|TextRender|SetText (Variables|Default|GetTextRender) (Variables|Default|GetNumber%20Debug))
              (elif (Utilities|Operators|Equal(==) (Variables|Default|GetCurrentRelativeRotation) "0.000000,270.000000,0.000000" 1.0)
                (Variables|Default|SetCurrentPuzzleData (Utilities|Struct|MakeBPPuzzleDataStruct 6 "0.000000,270.000000,0.000000"))
                (Variables|Default|SetNumber%20Debug "NSLOCTEXT(\"[31696A52E8C313A6E399A325CD403B52]\", \"3528F4914233AF2AEF5F7EBB63F5A195\", \"6\")")
                (Rendering|Components|TextRender|SetText (Variables|Default|GetTextRender) (Variables|Default|GetNumber%20Debug))
                (elif (Utilities|Operators|Equal(==) (Variables|Default|GetCurrentRelativeRotation) "0.000000,315.000000,0.000000" 1.0)
                  (Variables|Default|SetCurrentPuzzleData (Utilities|Struct|MakeBPPuzzleDataStruct 7 "0.000000,315.000000,0.000000"))
                  (Variables|Default|SetNumber%20Debug "NSLOCTEXT(\"[31696A52E8C313A6E399A325CD403B52]\", \"FBCB3F354F5CBFB7E618E786FE625BD8\", \"7\")")
                  (Rendering|Components|TextRender|SetText (Variables|Default|GetTextRender) (Variables|Default|GetNumber%20Debug))
                  (else
                    (Development|PrintString "Hello")))))))))))

(fn SetInitialPuzzleStartingIndex
  (bind (_puzzlepositionint_4_bd567a334c78235b071f6f897c39eb6f _puzzlerelativerotationyaw_5_5e7e8d054589dbc0540540b382104dac) (Utilities|Struct|BreakBPPuzzleDataStruct (Utilities|Array|Get(acopy) (Variables|Default|GetPuzzlePositionsStructure) (Math|Integer|Clamp(Integer) (Variables|Default|GetPuzzleStartingIndex) 0 7))))
  (CallFunction|SetCurrentandNextRotation)
  (Transformation|SetRelativeRotation (Variables|Default|GetStatueMesh) _puzzlerelativerotationyaw_5_5e7e8d054589dbc0540540b382104dac false false)
  (CallFunction|SetCurrentPuzzleDataIndex)
  (Variables|Default|SetNumber%20Debug (Utilities|Text|ToText(Integer) (Variables|Default|GetPuzzleStartingIndex))))

(fn ConstructionScript
  (bind (_puzzlepositionint_4_bd567a334c78235b071f6f897c39eb6f _puzzlerelativerotationyaw_5_5e7e8d054589dbc0540540b382104dac) (Utilities|Struct|BreakBPPuzzleDataStruct (Utilities|Array|Get(acopy) (Variables|Default|GetPuzzlePositionsStructure) (Variables|Default|GetPuzzleStartingIndex))))
  (Transformation|SetRelativeRotation (Variables|Default|GetStatueMesh) _puzzlerelativerotationyaw_5_5e7e8d054589dbc0540540b382104dac false false)
  (Variables|Default|SetNumber%20Debug (Utilities|Text|ToText(Integer) (Variables|Default|GetPuzzleStartingIndex)))
  (Rendering|Components|TextRender|SetText (Variables|Default|GetTextRender) (Variables|Default|GetNumber%20Debug)))
