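# Niagara burst effect for a puzzle turner's solved state, judging by the name, built from three emitters: Cone, CoronaLight and Decals.
# Cone scales up a mesh with sprites, deriving dark-glow and flare colors from particle color, with alpha adjusted by time of day and platform.
# CoronaLight is a single growing, fading sprite glow, while Decals bursts 16 partitioned, warm-tinted fading sprites.
# Curves drive color and fades, alongside dynamic material parameters.
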NiagaraSystem /Hermes/Blueprint/Prototypes/ChrisK/FX/NS_PuzzleTurner_Solved
dependencies {
 usedDataInterfaces [&NiagaraDataInterfaceCurve, &NiagaraDataInterfaceColorCurve, &NiagaraDataInterfacePlatformSet]
 usedDynamicInputs []
 usedModules [
  &self:SystemSpawnScript.SystemScriptSource.SystemScriptGraph.NiagaraNodeAssignment_2.SetVariables_#1
  &SystemState
 ]
}
emitters {
 Cone {
  emitterUpdateScript [
   &EmitterState()
   &SpawnBurst_Instantaneous(Use Spawn Probability=false, UseLoopCountLimit=false)
  ]
  particleSpawnScript [
   &InitializeParticle(
    Mesh Scale=&Multiply_VectorByFloat(Vector=<0.3 0.5 0.5>, Float=2)
    Mesh Scale Mode=::NewEnumerator4
    Sprite Size Mode=::NewEnumerator3
    Uniform Sprite Size=210
    UsePositionOffset=false
    Write Material Random=true
   )
   &InitialMeshOrientation(
    Mesh Orientation Mode=::NewEnumerator2
    Rotation=&RandomRangeVector(Minimum=<-1 0 0>, Maximum=<1 0 0>)
    Use Rotation Vector=true
   )
   &SystemLocation(MaskSpawnGroup=false, Offset=$System.SpawnLocation) {enabled=false}
  ]
  particleUpdateScript [
   &ParticleState()
   &CameraOffset(Camera Offset Amount=30) {module=CameraOffset001}
   &ScaleMeshSize(Scale Factor=&VectorFromFloat(Value=&FloatFromCurve(FloatCurve=DataInterface() {propertyValues {Curve {keys [
          {arriveTangent=2.0989, interpMode=RCIM_Cubic, leaveTangent=2.0989, tangentMode=RCTM_User, time=0, value=0.2}
          {interpMode=RCIM_Cubic, tangentMode=RCTM_User, time=1, value=1}
         ]}}}))) {module=ScaleMeshSize001}
   &DynamicMaterialParameters(
    Index 0 Param 1=0.5
    Index 0 Param 2=0.6
    Index 1 Param 1=-0.5
    Index 1 Param 2=1
    Module.Index0Float0Write=true
    Module.Index0Float1Write=true
    Module.Index0Float2Write=false
    Module.Index0Float3Write=false
    Module.Index1Float0Write=true
    Module.Index1Float1Write=true
    Module.Index1Float2Write=false
    Module.Index1Float3Write=false
    Write Parameter Index 1=::NewEnumerator1
   ) {module=DynamicMaterialParameters001}
   &ScaleColor(
    Scale Alpha=&FloatFromCurve(
     FloatCurve=DataInterface() {propertyValues {Curve {keys [
         {interpMode=RCIM_Cubic, tangentMode=RCTM_User, time=0, value=0}
         {arriveTangent=-5.5022, interpMode=RCIM_Cubic, leaveTangent=-5.5022, tangentMode=RCTM_User, time=0.07, value=1}
         {arriveTangent=-1.0753, interpMode=RCIM_Cubic, leaveTangent=-1.0753, time=0.2, value=0.4}
         {tangentMode=RCTM_User, time=1, value=0}
        ]}}}
     Scale Curve=&FNTimeOfDayScalar(Night Scale=2, Morning Scale=1, Evening Scale=1)
    )
    Scale RGB=&Multiply_VectorByFloat(
     Vector=&MakeVectorFromLinearColorRGB(Color=&ColorFromCurve(ColorCurve=DataInterface() {propertyValues {
         AlphaCurve {keys [<at 0 1 RCIM_Cubic>]}
         BlueCurve {keys [<at 0.0211 0.2322>, {interpMode=RCIM_Cubic, leaveTangent=1.6293, time=0.9865, value=0.1021}]}
         GreenCurve {keys [<at 0.0211 0.8781>, {interpMode=RCIM_Cubic, leaveTangent=-30.3699, time=0.9865, value=0.2331}]}
         RedCurve {keys [<at 0.0211 2>, {interpMode=RCIM_Cubic, leaveTangent=-80.5167, time=0.9865, value=0.4453}]}
        }}))
     Float=&Lerp_Float(
      A=3
      B=4
      Alpha=&FloatFromCurve(FloatCurve=DataInterface() {propertyValues {Curve {keys [
           {arriveTangent=3.8544, interpMode=RCIM_Cubic, leaveTangent=3.8544, tangentMode=RCTM_User, time=0, value=0}
           {interpMode=RCIM_Cubic, tangentMode=RCTM_User, time=0.6, value=1}
          ]}}})
     )
    )
    ScaleA=true
    ScaleRGB=true
   )
   /SetVariables(Particles.ColorDarkGlow:LinearColor=&MultiplyLinearColor(
     Linear Color=$Particles.Color
     Scale Factor=&MakeVector4_FromLinearColor(LinearColor=<rgba 0.0212 0.0212 0.0212 1>)
    )) {isSetParameters=true}
   /SetVariables(
    Particles.ColorFlare:LinearColor=&MultiplyLinearColor(
     Linear Color=$Particles.Color
     Scale Factor=&MakeVector4(
      X=1
      Y=1
      Z=1
      W=&FloatFromCurve(
       FloatCurve=DataInterface() {propertyValues {Curve {keys [{interpMode=RCIM_Cubic, tangentMode=RCTM_User, time=0.5, value=1}, {interpMode=RCIM_Cubic, tangentMode=RCTM_User, time=1, value=0}]}}}
       Scale Curve=&FN_PickFloatDependOnPlatform(TargetPlatformActive=1, TargetPlatformNoActive=0.65)
      )
     )
    )
    Particles.SpriteSizeFlare:Vector2f=&Multiply_Vector2D_ByFloat(Float=0.3, Vector2D=$Particles.SpriteSize)
   ) {isSetParameters=true}
  ]
  rendererClasses [&NiagaraMeshRendererProperties, &NiagaraSpriteRendererProperties]
  renderers [
   {rendererClass=&NiagaraMeshRendererProperties, rendererIndex=0}
   {rendererClass=&NiagaraSpriteRendererProperties, rendererIndex=1}
   {rendererClass=&NiagaraSpriteRendererProperties, rendererIndex=2}
  ]
 }
 CoronaLight {
  emitterUpdateScript [
   &EmitterState()
   &SpawnBurst_Instantaneous(Spawn Count=1, Use Spawn Probability=false)
  ]
  particleSpawnScript [
   &InitializeParticle(
    Lifetime=1.1
    Position Offset=<0 0 40>
    Sprite Rotation Mode=::NewEnumerator2
    Sprite Size Mode=::NewEnumerator3
    Uniform Sprite Size=80
    UsePositionOffset=false
    Write Material Random=true
   )
   &CameraOffset(Camera Offset Amount=100)
   &SystemLocation(MaskSpawnGroup=false, Offset=$System.SpawnLocation) {enabled=false}
  ]
  particleUpdateScript [
   &ParticleState()
   &ScaleColor(
    Scale Alpha=&FloatFromCurve(FloatCurve=DataInterface() {propertyValues {Curve {keys [{interpMode=RCIM_Cubic, leaveTangent=-0.6434, tangentMode=RCTM_Break, time=0, value=1}, <at 1 0>]}}})
    Scale RGB=&MakeVectorFromLinearColorRGB(Color=&ColorFromCurve(ColorCurve=DataInterface() {propertyValues {
        AlphaCurve {keys [<at 0 1 RCIM_Cubic>]}
        BlueCurve {keys [<at 0.0211 0.2322>, {interpMode=RCIM_Cubic, leaveTangent=1.6293, time=0.9865, value=0.1021}]}
        GreenCurve {keys [<at 0.0211 0.8781>, {interpMode=RCIM_Cubic, leaveTangent=-30.3699, time=0.9865, value=0.2331}]}
        RedCurve {keys [<at 0.0211 2>, {interpMode=RCIM_Cubic, leaveTangent=-80.5167, time=0.9865, value=0.4453}]}
       }}))
    ScaleA=true
    ScaleRGB=true
   )
   &ScaleSpriteSize(Uniform Curve Sprite Scale=DataInterface() {propertyValues {Curve {keys [
        {arriveTangent=1.5787, interpMode=RCIM_Cubic, leaveTangent=1.5787, tangentMode=RCTM_User, time=0, value=0.33}
        {arriveTangent=0.3268, interpMode=RCIM_Cubic, leaveTangent=0.3268, tangentMode=RCTM_User, time=0.9, value=1}
       ]}}}) {module=ScaleSpriteSize001}
   &DynamicMaterialParameters(
    Index 0 Param 1=&FloatFromCurve(
     FloatCurve=DataInterface() {propertyValues {Curve {keys [{arriveTangent=-3.7751, interpMode=RCIM_Cubic, leaveTangent=-3.7751, tangentMode=RCTM_User, time=0, value=1.5}, <at 1 0.25>]}}}
     Scale Curve=1.5
    )
    Index 0 Param 2=1
    Index 0 Param 3=1
    Index 1 Param 1=3
    Index 1 Param 2=0.1
    Index 1 Param 3=1
    Module.Index0Float0Write=true
    Module.Index0Float1Write=true
    Module.Index0Float2Write=true
    Module.Index0Float3Write=false
    Module.Index1Float0Write=true
    Module.Index1Float1Write=true
    Module.Index1Float2Write=true
    Module.Index1Float3Write=false
    Write Parameter Index 1=::NewEnumerator1
   )
  ]
  rendererClasses [&NiagaraSpriteRendererProperties]
  renderers [
   {rendererClass=&NiagaraSpriteRendererProperties}
   {rendererClass=&NiagaraSpriteRendererProperties, rendererIndex=1}
  ]
 }
 Decals {
  emitterUpdateScript [
   &EmitterState()
   &SpawnBurst_Instantaneous(Spawn Count=16, Use Spawn Probability=false, UseLoopCountLimit=true)
  ]
  particleSpawnScript [
   &PartitionParticles(Number Of Partitions=16)
   &InitializeParticle(
    Color=<rgba 2 0.5762 0.2373 1>
    Lifetime=1.35
    Sprite Size Mode=::NewEnumerator3
    Uniform Sprite Size=90
    UsePositionOffset=false
    Write Material Random=true
   ) {module=InitializeParticle001}
   &SystemLocation(MaskSpawnGroup=false, Offset=$System.SpawnLocation) {enabled=false}
  ]
  particleUpdateScript [
   &ParticleState()
   &CameraOffset(Camera Offset Amount=15)
   &DynamicMaterialParameters(
    Index 0 Param 1=&Lerp_Float(
     A=3
     B=1
     Alpha=&FloatFromCurve(FloatCurve=DataInterface() {propertyValues {Curve {keys [<at 0 1>,<at 1 0>]}}})
    )
    Index 0 Param 2=1
    Module.Index0Param0Write=true
    Module.Index0Param1Write=false
    Module.Index0Param2Write=false
    Module.Index0Param3Write=false
   ) {module=DynamicMaterialParameters001}
   &ScaleColor(
    Scale Alpha=&FloatFromCurve(FloatCurve=DataInterface() {propertyValues {Curve {keys [
         {arriveTangent=8.5587, interpMode=RCIM_Cubic, leaveTangent=8.5587, tangentMode=RCTM_Break, time=0, value=0}
         {arriveTangent=-2.6498, interpMode=RCIM_Cubic, leaveTangent=-2.6498, tangentMode=RCTM_User, time=0.1, value=1}
         {arriveTangent=-0.1099, interpMode=RCIM_Cubic, tangentMode=RCTM_Break, time=1, value=0}
        ]}}})
    ScaleA=true
    ScaleRGB=true
   ) {module=ScaleColor001}
  ]
  rendererClasses [&NiagaraSpriteRendererProperties]
  renderers [{rendererClass=&NiagaraSpriteRendererProperties}]
 }
}
summary {
 systemName=NS_PuzzleTurner_Solved
 emitters [
  {emitterName=Decals, rendererClasses [&NiagaraSpriteRendererProperties]}
  {emitterName=Cone, rendererClasses [&NiagaraMeshRendererProperties, &NiagaraSpriteRendererProperties]}
  {emitterName=CoronaLight, rendererClasses [&NiagaraSpriteRendererProperties]}
 ]
 userVariables []
}
